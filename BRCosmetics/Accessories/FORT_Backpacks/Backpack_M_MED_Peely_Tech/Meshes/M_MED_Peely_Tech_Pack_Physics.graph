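# Physics asset for the Peely Tech backpack cosmetic, defining bodies and joints for its dangling straps.
# A box covers the pack, with three capsule chains (two buckle straps, one right strap) attached by free-rotating master joints.
# Strap segments link with wide swing limits and very tight twist.
# An extra free-rotating constraint ties the shorter buckle strap's tip to the right strap's second segment, closing a loop through the pack.
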PhysicsAsset /BRCosmetics/Accessories/FORT_Backpacks/Backpack_M_MED_Peely_Tech/Meshes/M_MED_Peely_Tech_Pack_Physics
bodies {
 dyn_bucklestrap_1_jnt [{
   shapeType=Capsule
   length=2.75
   radius=2
   shapeName=dyn_bucklestrap_1_jnt_capsule
   center=<3.2946 -0.95 0>
   rotation=<rot -90 0.0197 -0.0197>
  }]
 dyn_bucklestrap_1_jnt1 [{
   shapeType=Capsule
   length=5.5
   radius=2
   shapeName=dyn_bucklestrap_1_jnt1_capsule
   center=<4.1165 0.0001 0.3819>
   rotation=<rot -79.9999 0 0>
  }]
 dyn_bucklestrap_2_jnt [{
   shapeType=Capsule
   length=3.5
   radius=2.25
   shapeName=dyn_bucklestrap_2_jnt_capsule
   center=<3.8246 -1.168 0>
   rotation=<rot -90 0.0264 -0.0263>
  }]
 dyn_bucklestrap_2_jnt2 [{
   shapeType=Capsule
   length=6.75
   radius=2
   shapeName=dyn_bucklestrap_2_jnt2_capsule
   center=<5.0574 -0.0003 -0.0745>
   rotation=<rot -90 -1.8984 1.8984>
  }]
 dyn_bucklestrap_3_jnt [{
   shapeType=Capsule
   length=4.5
   radius=3.75
   shapeName=dyn_bucklestrap_3_jnt_capsule
   center=<6.4142 -2.0283 0.0003>
   rotation=<rot -90 0.0264 -0.0263>
  }]
 dyn_bucklestrap_master_jnt [{shapeType=Capsule, length=1, radius=1, shapeName=dyn_bucklestrap_master_jnt_capsule}]
 dyn_bucklestrap_master_jnt1 [{
   shapeType=Capsule
   length=3
   radius=1.25
   shapeName=dyn_bucklestrap_master_jnt1_capsule
   center=<2.5549 -0.0001 0.4505>
   rotation=<rot -79.9998 -0.0001 0>
  }]
 dyn_rightstrap_1_jnt [{
   shapeType=Capsule
   length=3.25
   radius=3
   shapeName=dyn_rightstrap_1_jnt_capsule
   center=<4.0692 0.0001 0>
   rotation=<rot -90 0.0395 -0.0395>
  }]
 dyn_rightstrap_2_jnt [{shapeType=Capsule, length=4.75, radius=3, shapeName=dyn_rightstrap_2_jnt_capsule, center=<4.4668 0 0>, rotation=<rot -90 0.0395 -0.0395>}]
 dyn_rightstrap_3_jnt [{shapeType=Capsule, length=5.5, radius=3, shapeName=dyn_rightstrap_3_jnt_capsule, center=<5.3391 0 0>, rotation=<rot -90 -0.0395 0.0394>}]
 dyn_rightstrap_4_jnt [{shapeType=Capsule, length=4, radius=3, shapeName=dyn_rightstrap_4_jnt_capsule, center=<4.7348 0.0001 0>, rotation=<rot -90 0.0099 -0.0099>}]
 dyn_rightstrap_master_jnt [{shapeType=Capsule, length=1, radius=1, shapeName=dyn_rightstrap_master_jnt_capsule}]
 pack [{shapeType=Box, extentX=32.5, extentY=8.5, extentZ=62, shapeName=pack_box, center=<12.2329 5.0934 -19.241>, rotation=<rot 0 -10.0008 0>}]
}
constraints [
 {bone1Name=dyn_bucklestrap_1_jnt, bone2Name=dyn_bucklestrap_master_jnt, swing1LimitDegrees=60, swing2LimitDegrees=60, twistLimitDegrees=2}
 {bone1Name=dyn_bucklestrap_2_jnt, bone2Name=dyn_bucklestrap_1_jnt, swing1LimitDegrees=60, swing2LimitDegrees=60, twistLimitDegrees=2}
 {bone1Name=dyn_bucklestrap_3_jnt, bone2Name=dyn_bucklestrap_2_jnt, swing1LimitDegrees=60, swing2LimitDegrees=60, twistLimitDegrees=2}
 {bone1Name=dyn_rightstrap_1_jnt, bone2Name=dyn_rightstrap_master_jnt, swing1LimitDegrees=60, swing2LimitDegrees=60, twistLimitDegrees=2}
 {bone1Name=dyn_rightstrap_2_jnt, bone2Name=dyn_rightstrap_1_jnt, swing1LimitDegrees=60, swing2LimitDegrees=60, twistLimitDegrees=2}
 {bone1Name=dyn_rightstrap_3_jnt, bone2Name=dyn_rightstrap_2_jnt, swing1LimitDegrees=60, swing2LimitDegrees=60, twistLimitDegrees=2}
 {bone1Name=dyn_rightstrap_4_jnt, bone2Name=dyn_rightstrap_3_jnt, swing1LimitDegrees=60, swing2LimitDegrees=60, twistLimitDegrees=2}
 {bone1Name=dyn_bucklestrap_1_jnt1, bone2Name=dyn_bucklestrap_master_jnt1, swing1LimitDegrees=60, swing2LimitDegrees=60, twistLimitDegrees=2}
 {bone1Name=dyn_bucklestrap_2_jnt2, bone2Name=dyn_bucklestrap_1_jnt1, swing1LimitDegrees=60, swing2LimitDegrees=60, twistLimitDegrees=2}
 {
  bone1Name=dyn_bucklestrap_master_jnt
  bone2Name=pack
  swing1LimitDegrees=45
  swing1Motion=Free
  swing2LimitDegrees=45
  swing2Motion=Free
  twistLimitDegrees=45
  twistMotion=Free
 }
 {
  bone1Name=dyn_rightstrap_master_jnt
  bone2Name=pack
  swing1LimitDegrees=45
  swing1Motion=Free
  swing2LimitDegrees=45
  swing2Motion=Free
  twistLimitDegrees=45
  twistMotion=Free
 }
 {
  bone1Name=dyn_bucklestrap_master_jnt1
  bone2Name=pack
  swing1LimitDegrees=45
  swing1Motion=Free
  swing2LimitDegrees=45
  swing2Motion=Free
  twistLimitDegrees=45
  twistMotion=Free
 }
 {
  bone1Name=dyn_bucklestrap_2_jnt2
  bone2Name=dyn_rightstrap_2_jnt
  swing1LimitDegrees=45
  swing1Motion=Free
  swing2LimitDegrees=45
  swing2Motion=Free
  twistLimitDegrees=45
  twistMotion=Free
 }
]
